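# Physics asset for the Future Biker White two-handed pickaxe, defining collision bodies and joint limits for its dynamic chain.
# A capsule covers the handle, capsules cover chain bones 07 to 010, and boxes cover bones 01, 02 and the master bone.
# Limited-angle joints form two runs, master-01-02 and handle-010-09-08-07.
# An angularly locked constraint between 02 and 07 joins the runs, linking master to handle.
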PhysicsAsset /BRCosmetics/Weapons/FORT_Melee/Pickaxe_FutureBikerWhite/Meshes/Future_Biker_White_Axe_PhysicsAsset
bodies {
 Melee_TwoHanded_Handle [{shapeType=Capsule, length=120.9247, radius=2.6165, center=<1.0844 -1.5794 25.447>, rotation=<rot 1.0492 -90 -0.0549>}]
 dyn_chain_01 [{shapeType=Box, extentX=7.86, extentY=1.4914, extentZ=7.8042, center=<3.4275 -0.0241 0.0937>, rotation=<rot 0 93.3233 -86.2444>}]
 dyn_chain_010 [{shapeType=Capsule, length=5, radius=2, center=<3 0 0>, rotation=<rot 90 0 0>}]
 dyn_chain_02 [{shapeType=Box, extentX=1.0912, extentY=3.2527, extentZ=4.984, center=<1.7497 -0.1933 0.0567>, rotation=<rot 0 -89.7638 -90.4596>}]
 dyn_chain_07 [{shapeType=Capsule, length=3.8659, radius=1.874, center=<2.5241 -0.0298 -0.5503>, rotation=<rot 0 -92.3973 -67.899>}]
 dyn_chain_08 [{shapeType=Capsule, length=4.1537, radius=1.8525, center=<3.1048 -0.1377 -0.5185>, rotation=<rot 0 -89.1313 -71.7629>}]
 dyn_chain_09 [{shapeType=Capsule, length=4.1044, radius=1.6941, center=<3.3103 -0.0764 -0.2793>, rotation=<rot 0 -91.3082 -83.6671>}]
 dyn_chain_mstr [{shapeType=Box, extentX=7.5, extentY=7.5, extentZ=7.5}]
}
constraints [
 {bone1Name=dyn_chain_01, bone2Name=dyn_chain_mstr, swing1LimitDegrees=5, swing2LimitDegrees=15, twistLimitDegrees=5}
 {bone1Name=dyn_chain_02, bone2Name=dyn_chain_01, swing1LimitDegrees=30, swing2LimitDegrees=30, twistLimitDegrees=5}
 {bone1Name=dyn_chain_010, bone2Name=Melee_TwoHanded_Handle, swing1LimitDegrees=45, swing2LimitDegrees=45, twistLimitDegrees=5}
 {bone1Name=dyn_chain_09, bone2Name=dyn_chain_010, swing1LimitDegrees=45, swing2LimitDegrees=45, twistLimitDegrees=5}
 {bone1Name=dyn_chain_08, bone2Name=dyn_chain_09, swing1LimitDegrees=45, swing2LimitDegrees=45, twistLimitDegrees=5}
 {bone1Name=dyn_chain_07, bone2Name=dyn_chain_08, swing1LimitDegrees=45, swing2LimitDegrees=45, twistLimitDegrees=5}
 {
  bone1Name=dyn_chain_02
  bone2Name=dyn_chain_07
  swing1LimitDegrees=2
  swing1Motion=Locked
  swing2LimitDegrees=2
  swing2Motion=Locked
  twistLimitDegrees=2
  twistMotion=Locked
 }
]
