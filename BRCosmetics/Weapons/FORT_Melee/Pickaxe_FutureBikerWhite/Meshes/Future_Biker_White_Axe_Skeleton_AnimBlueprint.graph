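# Animation Blueprint for the Future Biker White pickaxe skeleton that adds secondary physics motion to the weapon.
# The input pose is cached and fed through a RigidBody node using the axe's physics asset, simulated relative to the two-handed handle bone with custom gravity and damped component motion.
# A branch of dyn_chain bone constraints and a layered blend does not reach the output pose.
# The event graph holds only unconnected stub nodes.
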AnimBlueprint /BRCosmetics/Weapons/FORT_Melee/Pickaxe_FutureBikerWhite/Meshes/Future_Biker_White_Axe_Skeleton_AnimBlueprint
target_skeleton=&/Future_Biker_White_Axe_Skeleton
graphs [
 AnimGraph {nodes [
   n1 = Component To Local()
   n2 = "Constraint - dyn_chain_09 (T)"(ComponentPose=Local To Component(LocalPose=Use cached pose 'InputPose'()), Alpha=0, ConstraintWeights_0=1) {
    alphaBoolBlend {blendOption=Linear}
    boneToModify {boneName=dyn_chain_09}
    constraintSetup [{transformType=Translation, targetBone {boneName=dyn_chain_010}}]
    constraintWeights [1]
   }
   n3 = Component To Local(ComponentPose=@n2)
   "Constraint - dyn_chain_010 (P)"(
    ComponentPose=Local To Component(LocalPose=Layered blend per bone(BasePose=@n1, BlendPoses_0=@n3, BlendWeights_0=0) {layerSetup [{branchFilters [{blendDepth=1, boneName=dyn_chain_010}]}]})
    Alpha=1
    ConstraintWeights_0=1
   ) {
    alphaBoolBlend {blendOption=Linear}
    boneToModify {boneName=dyn_chain_010}
    constraintSetup [{targetBone {boneName=dyn_chain_constraint}}]
    constraintWeights [1]
   }
   n4 = RigidBody(ComponentPose=Local To Component(LocalPose=Use cached pose 'InputPose'())) {
    bClampLinearTranslationLimitToRefPose=true
    bOverrideWorldGravity=true
    overridePhysicsAsset=&/Future_Biker_White_Axe_PhysicsAsset
    simulationSpace=BaseBoneSpace
    alphaBoolBlend {blendOption=Linear}
    baseBoneRef {boneName=Melee_TwoHanded_Handle}
    componentAppliedLinearAccClamp=<3000 3000 3000>
    componentLinearAccScale=<0.3 0.3 0.3>
    componentLinearVelScale=<0.3 0.3 0.3>
    overrideWorldGravity {z=-580}
   }
   Output Pose(Result=Component To Local(ComponentPose=@n4)) {layerGroup=None}
   AnimGraphNode_SubInput_0 = Input Pose()
   Save cached pose 'InputPose'(Pose=@AnimGraphNode_SubInput_0) {cacheName=InputPose}
  ]}
 EventGraph {nodes [
   TryGetPawnOwner()
   Event BlueprintUpdateAnimation()
  ]}
]
variables []
