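# Idle glow effect for the SpecialScroll pickaxe under Pickaxe_FutureBikerWhite, built from persistent pale-blue burst-spawned sprites.
# MainGlow places three oriented sprites at fixed offsets by spawn group, tagged for visibility across three sprite renderers.
# NE_SupportingGlow001 partitions eight camera-offset sprites, each following its own socket through a system Socket Reader.
# Both emitters scale alpha by time of day and front-end state; the supporting glow also varies size and alignment in front end.
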NiagaraSystem /BRCosmetics/Weapons/FORT_Melee/Pickaxe_FutureBikerWhite/Skins/FX/NS_Pickaxe_SpecialScroll_Idle
dependencies {
 usedDataInterfaces [
  &NiagaraDataInterfaceArrayFloat3
  &NiagaraDataInterfaceArrayInt32
  &NiagaraDataInterfaceSocketReader
 ]
 usedDynamicInputs []
 usedModules [
  &self:SystemSpawnScript.SystemScriptSource.SystemScriptGraph.NiagaraNodeAssignment_1.SetVariables_#1
  &self:SystemSpawnScript.SystemScriptSource.SystemScriptGraph.NiagaraNodeAssignment_0.SetVariables_#2
  &SystemState
 ]
}
emitters {
 MainGlow {
  emitterUpdateScript [
   &EmitterState(
    Inactive Response=::NewEnumerator1
    Life Cycle Mode=::NewEnumerator1
    Loop Behavior=::NewEnumerator1
    Loop Duration Mode=::NewEnumerator1
    UseLoopDelay=false
   )
   &SpawnBurst_Instantaneous(Spawn Count=3, Use Spawn Probability=false, UseLoopCountLimit=false)
  ]
  particleSpawnScript [
   &FN_SetSpawnGroup(Number of Spawn Groups=3)
   &InitializeParticle(
    Color=&MultiplyLinearColor(
     Linear Color=<rgba 0.4 0.7 1 0.2>
     Scale Factor=&MakeVector4FromVectorAndFloat(
      "Vector (RGB)"=<1 1 1>
      "Float (Alpha)"=&MergeHelper_FloatSwitch(0=0.7, 1=3, 2=3, 3=?, Index=&SpawnGroup(), Num=::NewEnumerator1)
     )
    )
    Position Offset=&MergeHelper_VectorSwitch(0=<1 14 45>, 1=<1 12 57.5>, 2=<1 6 48.7768>, 3=?, Index=&SpawnGroup(), Num=::NewEnumerator1)
    Sprite Rotation Angle=&MergeHelper_FloatSwitch(0=-1, 1=0, 2=0, 3=?, Index=&SpawnGroup(), Num=::NewEnumerator1)
    Sprite Rotation Mode=::NewEnumerator0
    Sprite Size=&MergeHelper_Vector2Switch(
     0=&Multiply_Vector2D_ByFloat(Float=-1, Vector2D=<97 93>)
     1=<100 12>
     2=<35 40>
     3=?
     4=?
     Index=&SpawnGroup()
     Num=::NewEnumerator1
    )
    Sprite Size Mode=::NewEnumerator4
    UsePositionOffset=true
    Write Material Random=true
   )
  ]
  particleUpdateScript [
   &ParticleState(Kill Particles When Lifetime Has Elapsed=false, Let Infinitely Lived Particles Die When Emitter Deactivates=true)
   &SpriteFacingAndAlignment(
    Sprite Alignment=&MergeHelper_VectorSwitch(0=<0 0 -1>, 1=<1 0 0>, 2=<0 0 1>, 3=?, Index=&SpawnGroup(), Num=::NewEnumerator1)
    Sprite Facing=&MergeHelper_VectorSwitch(0=<1 0 0>, 1=<0 0.2561 1>, 2=<0 1 0>, 3=?, Index=&SpawnGroup(), Num=::NewEnumerator1)
    WriteAlignment=true
    WriteFacing=true
   )
   &DynamicMaterialParameters(Module.Param0WriteEnabled=true, Module.Param1WriteEnabled=true, Module.Param2WriteEnabled=true, Module.Param3WriteEnabled=true)
   &ScaleColor(
    Scale Alpha=&Multiply_Float(
     A=&FNTimeOfDayScalar(Night Scale=0.8, Morning Scale=0.8, Evening Scale=0.8)
     B=&FN_FrontEndFloatModifier(Gameplay=1, Front End=0.7)
    )
    ScaleA=true
    ScaleRGB=true
   )
   /SetVariables(Particles.VisibilityTag=&SpawnGroup()) {isSetParameters=true}
  ]
  rendererClasses [&NiagaraSpriteRendererProperties]
  renderers [
   {rendererClass=&NiagaraSpriteRendererProperties}
   {rendererClass=&NiagaraSpriteRendererProperties, rendererIndex=1}
   {rendererClass=&NiagaraSpriteRendererProperties, rendererIndex=2}
  ]
 }
 NE_SupportingGlow001 {
  emitterUpdateScript [
   &EmitterState(
    Inactive Response=::NewEnumerator1
    Life Cycle Mode=::NewEnumerator1
    Loop Behavior=::NewEnumerator1
    Loop Duration Mode=::NewEnumerator1
    UseLoopDelay=false
   )
   &SpawnBurst_Instantaneous(Spawn Count=8, Use Spawn Probability=false, UseLoopCountLimit=false)
  ]
  particleSpawnScript [
   &PartitionParticles(Number Of Partitions=8)
   &InitializeParticle(
    Color=&MergeHelper_ColorSwitch(
     0=<rgba 0.7 1.05 1.4 0.2>
     1=<rgba 0.7 1.05 1.4 0.2>
     2=<rgba 0.7 1.05 1.4 0.2>
     3=<rgba 0.7 1.05 1.4 0.1>
     4=<rgba 0.7 1.05 1.4 0.1>
     5=<rgba 0.7 1.05 1.4 0.2>
     6=<rgba 0.7 1.05 1.4 0.2>
     7=<rgba 0.7 1.05 1.4 0.2>
     Index=$Particles.PartitionParticles.Partition
     Num=::NewEnumerator6
    )
    Sprite Size=&MergeHelper_Vector2Switch(
     0=<55 105>
     1=<15 22>
     2=<15 22>
     3=&FN_FrontEndVector2Modifier(Front End:Vector2f=<100 100>, Gameplay:Vector2f=<40 80>)
     4=<30 55>
     5=<22 22>
     6=<22 37.1569>
     7=<22 22>
     Index=$Particles.PartitionParticles.Partition
     Num=::NewEnumerator6
    )
    Sprite Size Mode=::NewEnumerator4
    UsePositionOffset=false
    Write Material Random=true
   )
   &CameraOffset(Camera Offset Amount=3)
  ]
  particleUpdateScript [
   &ParticleState(Kill Particles When Lifetime Has Elapsed=false, Let Infinitely Lived Particles Die When Emitter Deactivates=true)
   &SpriteFacingAndAlignment(
    Sprite Alignment=&FN_FrontEndVector3Modifier(
     Gameplay=&SelectVectorFromArray(
      Direct Array Index=$Particles.PartitionParticles.Partition
      Vector Selection Array=DataInterface() {propertyValues {FloatData [<0 0 1>, <0.25 1 0>, <0.05 1 0>, <0 -0.55 1>, <0 0 1>, <0 0 1>, <0 0 1>, <0 0 1>]}}
     )
     Front End=&SelectVectorFromArray(
      Direct Array Index=$Particles.PartitionParticles.Partition
      Vector Selection Array=DataInterface() {propertyValues {FloatData [<0 0 1>, <0.25 1 0>, <0.05 1 0>, <0 0 1>, <0 0 1>, <0 0 1>, <0 0 1>, <0 0 1>]}}
     )
    )
    WriteAlignment=true
    WriteFacing=false
   )
   &SocketLocation(
    SamplePosOffset=false
    Socket Index=&SelectIntFromArray(Direct Array Index=$Particles.PartitionParticles.Partition, Int Selection Array=DataInterface() {propertyValues {IntData [0,1,2,3,4,5,6,7]}})
    Socket Reader=$System.Socket Reader
    Socket Sampling Mode=::NewEnumerator3
    VelocitySpeedLimit=false
   )
   &ScaleColor(Color Scale:LinearColor=<rgba 1 1 1 0.5>, Scale Mode=3)
   &ScaleColor(
    Scale Alpha=&MakeCustomFloatFromBool(
     Bool=$User.bIsFrontEndPreview
     True Float=0.15
     False Float=&FNTimeOfDayScalar(Night Scale=0.2, Morning Scale=0.15, Day Scale=0.35, Evening Scale=0.15)
    )
    ScaleA=true
    ScaleRGB=true
   ) {module=ScaleColor001}
  ]
  rendererClasses [&NiagaraSpriteRendererProperties]
  renderers [{rendererClass=&NiagaraSpriteRendererProperties}]
 }
}
summary {
 systemName=NS_Pickaxe_SpecialScroll_Idle
 emitters [
  {emitterName=MainGlow, rendererClasses [&NiagaraSpriteRendererProperties]}
  {emitterName=NE_SupportingGlow001, rendererClasses [&NiagaraSpriteRendererProperties]}
 ]
 userVariables [
  User.bIsFrontEnd {type=NiagaraBool}
  User.bIsFrontEndPreview {type=NiagaraBool}
  User.bisSwinging {type=NiagaraBool}
  User.FrontEndScale {type=NiagaraFloat, defaultValue=0.55, description=""}
  User.PhysicalSurface {type=NiagaraInt32}
  User.RibbonIDOffset {type=NiagaraInt32}
 ]
}
